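# Niagara plasma swing-trail effect for the FairyFlex pickaxe, combining ribbon trails with ember sprites.
# EnergyTrail and Trail_Gold build anim-trail ribbons from skeletal mesh bone/socket pairs, with time-of-day scaling; EnergyTrail adds a second renderer, time-driven U offset and dark secondary color.
# Near-duplicate GPU and CPU ember emitters spawn per unit moved around a mesh socket, inherit capped velocity, then shrink and fade.
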NiagaraSystem /BRCosmetics/Weapons/FORT_Melee/Pickaxe_FairyFlex/FX/Plasma/NS_Pickaxe_FairyFlex_Plasm
dependencies {
 usedDataInterfaces [&NiagaraDataInterfaceColorCurve, &NiagaraDataInterfaceCurve, &NiagaraDataInterfaceSkeletalMesh]
 usedDynamicInputs []
 usedModules [
  &self:SystemSpawnScript.SystemScriptSource.SystemScriptGraph.NiagaraNodeAssignment_0.SetVariables_#1
  &SystemState
 ]
}
emitters {
 Embers {
  simTarget=GPUComputeSim
  emitterUpdateScript [
   &EmitterState()
   &SpawnPerUnit(Module.Use Max Movement Threshold=true, Module.Use Movement Tolerance=false, Use Spawn Probability=false)
  ]
  particleSpawnScript [
   &InitializeParticle(
    Color Mode=::NewEnumerator0
    Lifetime Max=0.3
    Lifetime Min=0.15
    Lifetime Mode=::NewEnumerator1
    Sprite Size Max=<2.5 2>
    Sprite Size Min=<4 3>
    Sprite Size Mode=::NewEnumerator2
    UsePositionOffset=false
    Write Material Random=true
   )
   &SkeletalMeshLocation(
    Bone / Socket Index=3
    MaskSpawnGroup=false
    Mesh Sampling Type=::NewEnumerator5
    Skeletal Mesh=$System.Skeletal Mesh
    Socket Sampling Mode=::NewEnumerator5
   )
   &ShapeLocation(
    MaskSpawnGroup=false
    Offset Coordinate Space=::Local
    Offset Mode=::NewEnumerator0
    Override Seed=false
    Shape Origin=$Particles.SkeletalMeshLocation.SampledPosition
    Sphere Radius=&RandomRangeFloat(Minimum=5, Maximum=10)
    UseOwnerScale=true
   )
   &InheritVelocity(Inherited Velocity Speed Limit=150, Speed Limit=true, UseSpeedThreshold=false)
   &ApplyInitialForces()
  ]
  particleUpdateScript [
   &Drag(Drag=&RandomRangeFloat(Minimum=4, Maximum=6), Use Linear Drag=true, Use Rotational Drag=true)
   &SolveForcesAndVelocity(Clamp Velocity=false, Limit Acceleration=false)
   &ParticleState()
   &ScaleSpriteSize(Uniform Curve Sprite Scale=DataInterface() {propertyValues {Curve {keys [
        {interpMode=RCIM_Cubic, tangentMode=RCTM_User, time=0, value=0}
        <at 0.2 1>
        {interpMode=RCIM_Cubic, tangentMode=RCTM_User, time=0.7, value=1}
        {tangentMode=RCTM_User, time=1, value=0}
       ]}}})
   &ScaleSpriteSizeBySpeed(Max Scale Factor=<0.5 1.8>, Min Scale Factor=<0.5 0.9>, Velocity Threshold=150)
   &ScaleColor(
    Scale RGB=&MakeVectorFromLinearColorRGB(Color=&ColorFromCurve(ColorCurve=DataInterface() {propertyValues {
        AlphaCurve {keys [<at 0 1 RCIM_Cubic>,<at 0.9895 0>]}
        BlueCurve {keys [<at 0.232 0.0104>, {arriveTangent=-0.0136, interpMode=RCIM_Cubic, leaveTangent=-0.0136, time=0.5666, value=0}, <at 1 0 RCIM_Cubic>]}
        GreenCurve {keys [<at 0.232 0.9505>, {arriveTangent=0.0644, interpMode=RCIM_Cubic, leaveTangent=0.0644, time=0.5666, value=0.4896}, <at 1 1 RCIM_Cubic>]}
        RedCurve {keys [<at 0.232 1>, {arriveTangent=-0.9548, interpMode=RCIM_Cubic, leaveTangent=-0.9548, time=0.5666, value=0.2611}, <at 1 0.2667 RCIM_Cubic>]}
       }}))
    ScaleA=true
    ScaleRGB=true
   )
  ]
  rendererClasses [&NiagaraSpriteRendererProperties]
  renderers [{rendererClass=&NiagaraSpriteRendererProperties}]
 }
 Embers001 {
  emitterUpdateScript [
   &EmitterState()
   &SpawnPerUnit(Module.Use Max Movement Threshold=true, Module.Use Movement Tolerance=false, Spawn Spacing=40, Use Spawn Probability=false)
  ]
  particleSpawnScript [
   &InitializeParticle(
    Color Mode=::NewEnumerator0
    Lifetime Max=0.3
    Lifetime Min=0.15
    Lifetime Mode=::NewEnumerator1
    Sprite Size Max=<2.5 2>
    Sprite Size Min=<4 3>
    Sprite Size Mode=::NewEnumerator2
    UsePositionOffset=false
    Write Material Random=true
   )
   &SkeletalMeshLocation(
    Bone / Socket Index=3
    MaskSpawnGroup=false
    Mesh Sampling Type=::NewEnumerator5
    Skeletal Mesh=$System.Skeletal Mesh
    Socket Sampling Mode=::NewEnumerator5
   )
   &ShapeLocation(
    MaskSpawnGroup=false
    Offset Coordinate Space=::Local
    Offset Mode=::NewEnumerator0
    Override Seed=false
    Shape Origin=$Particles.SkeletalMeshLocation.SampledPosition
    Sphere Radius=&RandomRangeFloat(Minimum=5, Maximum=10)
    UseOwnerScale=true
   )
   &InheritVelocity(Inherited Velocity Speed Limit=150, Speed Limit=true, UseSpeedThreshold=false)
   &ApplyInitialForces()
  ]
  particleUpdateScript [
   &Drag(Drag=&RandomRangeFloat(Minimum=4, Maximum=6), Use Linear Drag=true, Use Rotational Drag=true)
   &SolveForcesAndVelocity(Clamp Velocity=false, Limit Acceleration=false)
   &ParticleState()
   &ScaleSpriteSize(Uniform Curve Sprite Scale=DataInterface() {propertyValues {Curve {keys [
        {interpMode=RCIM_Cubic, tangentMode=RCTM_User, time=0, value=0}
        <at 0.2 1>
        {interpMode=RCIM_Cubic, tangentMode=RCTM_User, time=0.7, value=1}
        {tangentMode=RCTM_User, time=1, value=0}
       ]}}})
   &ScaleSpriteSizeBySpeed(Max Scale Factor=<0.5 1.8>, Min Scale Factor=<0.5 0.9>, Velocity Threshold=150)
   &ScaleColor(
    Scale RGB=&MakeVectorFromLinearColorRGB(Color=&ColorFromCurve(ColorCurve=DataInterface() {propertyValues {
        AlphaCurve {keys [<at 0 1 RCIM_Cubic>,<at 0.9895 0>]}
        BlueCurve {keys [<at 0.232 0.0104>, {arriveTangent=-0.0136, interpMode=RCIM_Cubic, leaveTangent=-0.0136, time=0.5666, value=0}, <at 1 0 RCIM_Cubic>]}
        GreenCurve {keys [<at 0.232 0.9505>, {arriveTangent=0.0644, interpMode=RCIM_Cubic, leaveTangent=0.0644, time=0.5666, value=0.4896}, <at 1 1 RCIM_Cubic>]}
        RedCurve {keys [<at 0.232 1>, {arriveTangent=-0.9548, interpMode=RCIM_Cubic, leaveTangent=-0.9548, time=0.5666, value=0.2611}, <at 1 0.2667 RCIM_Cubic>]}
       }}))
    ScaleA=true
    ScaleRGB=true
   )
  ]
  rendererClasses [&NiagaraSpriteRendererProperties]
  renderers [{rendererClass=&NiagaraSpriteRendererProperties}]
 }
 EnergyTrail {
  emitterUpdateScript [
   &EmitterState()
   &FN_SpawnAnimTrail(Spawn Interval=0)
  ]
  particleSpawnScript [
   &FN_InitializeAnimTrailBoneIndecies(First Bone or Socket Index=3, Second Bone or Socket Index=4)
   &FN_InitializeAnimTrail(Lifetime=0.22, SkeletalMesh=$System.Skeletal Mesh)
  ]
  particleUpdateScript [
   &FN_UpdateAnimTrail(SkeletalMesh=$System.Skeletal Mesh)
   &ScaleColor(
    Scale Alpha=&FloatFromCurve(
     FloatCurve=DataInterface() {propertyValues {Curve {keys [
         {arriveTangent=-1.6414, interpMode=RCIM_Cubic, leaveTangent=-1.6414, tangentMode=RCTM_User, time=0, value=1}
         {interpMode=RCIM_Cubic, tangentMode=RCTM_User, time=1, value=0}
        ]}}}
     Scale Curve=0.5
    )
    Scale RGB=&Multiply_Vector(
     A=&MakeVectorFromLinearColorRGB(Color=&ColorFromCurve(ColorCurve=DataInterface() {propertyValues {
         AlphaCurve {keys [<at 0 1 RCIM_Cubic>,<at 1 1 RCIM_Cubic>]}
         BlueCurve {keys [<at 0.0021 0.0521>, {arriveTangent=-0.089, interpMode=RCIM_Cubic, leaveTangent=-0.089, time=0.234, value=0.2}, <at 0.5871 0>, <at 1 0>]}
         GreenCurve {keys [<at 0.0021 10>, {arriveTangent=-15.7241, interpMode=RCIM_Cubic, leaveTangent=-15.7241, time=0.234, value=2}, <at 0.5871 0.8>, <at 1 0.5>]}
         RedCurve {keys [
           <at 0.0021 4.6944>
           {arriveTangent=-6.7701, interpMode=RCIM_Cubic, leaveTangent=-6.7701, time=0.234, value=1.37}
           <at 0.5871 0.7333>
           <at 1 0.3167>
          ]}
        }}))
     B=&VectorFromFloat(Value=&FNTimeOfDayScalar(Night Scale=0.15, Morning Scale=0.7, Day Scale=0.5, Evening Scale=0.15))
    )
    ScaleA=true
    ScaleRGB=true
   )
   &ParticleState()
   /SetVariables(Particles.RibbonU0Override=&Multiply_Float(A=$Engine.Time, B=4)) {isSetParameters=true}
   &CameraOffset(Camera Offset Amount=10)
   &ScaleRibbonWidth(Ribbon Width Scale=&FloatFromCurve(FloatCurve=DataInterface() {propertyValues {Curve {keys [
         {arriveTangent=-2.083, interpMode=RCIM_Cubic, leaveTangent=-2.083, tangentMode=RCTM_User, time=0.25, value=1}
         {tangentMode=RCTM_User, time=1, value=0.25}
        ]}}}))
   /SetVariables(
    Particles.SecondaryColor:LinearColor=&MultiplyLinearColor(
     Linear Color=<rgba 0.02 0.02 0.02 0.8>
     Scale Factor=&MakeVector4(
      X=1
      Y=1
      Z=1
      W=&Multiply_Float(
       A=&MakeFloatFromLinearColor(LinearColor=$Particles.Color, Channel:ENiagaraLinearColor_Channels=::NewEnumerator3)
       B=&Multiply_Float(A=5, B=&FNTimeOfDayScalar(Night Scale=1, Morning Scale=1, Evening Scale=1.5))
      )
     )
    )
    Particles.SecondaryWidth:NiagaraFloat=&Multiply_Float(A=$Particles.RibbonWidth, B=1.1)
   ) {isSetParameters=true}
  ]
  rendererClasses [&NiagaraRibbonRendererProperties]
  renderers [
   {rendererClass=&NiagaraRibbonRendererProperties}
   {rendererClass=&NiagaraRibbonRendererProperties, rendererIndex=1}
  ]
 }
 Trail_Gold {
  emitterUpdateScript [
   &EmitterState()
   &FN_SpawnAnimTrail_v2()
  ]
  particleSpawnScript [
   /Game/Effects/Niagara/Modules/AnimTrail/FN_InitializeAnimTrail/FN_InitializeAnimTrailBoneIndecies(Second Bone or Socket Index=1)
   &FN_InitializeAnimTrail(Color=<rgba 0.7164 1 0.052 1>, Lifetime=0.22, SkeletalMesh=$System.Skeletal Mesh) {module=FN_InitializeAnimTrail001}
  ]
  particleUpdateScript [
   &FN_UpdateAnimTrail(SkeletalMesh=$System.Skeletal Mesh)
   &ScaleRibbonWidth(Ribbon Width Scale=&FloatFromCurve(FloatCurve=DataInterface() {propertyValues {Curve {keys [
         {arriveTangent=-1.2579, interpMode=RCIM_Cubic, time=0, value=1}
         {arriveTangent=-0.9664, interpMode=RCIM_Cubic, leaveTangent=-0.9664, tangentMode=RCTM_User, time=1, value=0}
        ]}}}))
   &ScaleColor(
    Scale Alpha=&Multiply_Float(
     A=&FloatFromCurve(FloatCurve=DataInterface() {propertyValues {Curve {keys [
          {interpMode=RCIM_Cubic, leaveTangent=-2.5656, tangentMode=RCTM_Break, time=0, value=1}
          {arriveTangent=-0.151, tangentMode=RCTM_User, time=0.5, value=0}
         ]}}})
     B=&FNTimeOfDayScalar(Night Scale=0.5, Morning Scale=0.75, Evening Scale=0.75)
    )
    Scale RGB=&VectorFromFloat(Value=&FNTimeOfDayScalar(Night Scale=0.5, Morning Scale=0.75, Evening Scale=0.75))
    ScaleA=true
    ScaleRGB=true
   ) {module=ScaleColor002}
  ]
  rendererClasses [&NiagaraRibbonRendererProperties]
  renderers [{rendererClass=&NiagaraRibbonRendererProperties}]
 }
}
summary {
 systemName=NS_Pickaxe_FairyFlex_Plasm
 emitters [
  {emitterName=EnergyTrail, rendererClasses [&NiagaraRibbonRendererProperties]}
  {emitterName=Embers, simTarget=GPUComputeSim, rendererClasses [&NiagaraSpriteRendererProperties]}
  {emitterName=Trail_Gold, rendererClasses [&NiagaraRibbonRendererProperties]}
  {emitterName=Embers001, rendererClasses [&NiagaraSpriteRendererProperties]}
 ]
 userVariables [User.RibbonIDOffset {type=NiagaraInt32}]
}
